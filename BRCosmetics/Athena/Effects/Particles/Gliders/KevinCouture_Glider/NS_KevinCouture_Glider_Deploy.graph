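# Niagara system for the KevinCouture glider deploy effect: a timed violet lightning burst with glow.
# AirGlow and Corona each flash a single sprite, followed shortly by six lightning meshes with particle lights.
# Three delayed spawn-rate emitters add crackling sprites, one also rendering meshes with decal-named attributes.
# Lightning particles use five spawn groups to follow bones or sockets on User.SkeletalMesh, animated through curve-driven color and dynamic material parameters.
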NiagaraSystem /BRCosmetics/Athena/Effects/Particles/Gliders/KevinCouture_Glider/NS_KevinCouture_Glider_Deploy
dependencies {
 usedDataInterfaces [&NiagaraDataInterfaceCurve, &NiagaraDataInterfaceColorCurve]
 usedDynamicInputs []
 usedModules [&SystemState]
}
emitters {
 AirGlow {
  emitterUpdateScript [
   &EmitterState(Life Cycle Mode=::NewEnumerator1, Loop Behavior=::NewEnumerator1, Loop Delay=0.5, Loop Duration=1, UseLoopDelay=false)
   &SpawnBurst_Instantaneous(Spawn Count=1, Spawn Time=0.65, Use Spawn Probability=false)
  ]
  particleSpawnScript [/Niagara/Modules/Spawn/Initialization/InitializeParticle(
    Color=<rgba 0.8317 0 2 1>
    Lifetime=0.2
    Module.Write Color=true
    Module.Write Lifetime=true
    Module.Write Mass=false
    Module.Write Material Random=true
    Module.Write Position=false
    Module.Write Scale=false
    Module.Write SpriteRotation=false
    Module.Write SpriteSize=true
    Position=&SimulationPosition()
    Sprite Size=<200 200>
   )]
  particleUpdateScript [
   &ParticleState()
   &ScaleColor(
    Scale Alpha=&FloatFromCurve(FloatCurve=DataInterface() {propertyValues {Curve {keys [
         <at 0 0 RCIM_Cubic>
         {arriveTangent=0.4483, interpMode=RCIM_Cubic, leaveTangent=0.4483, tangentMode=RCTM_User, time=0.5, value=1}
         <at 1 0 RCIM_Cubic>
        ]}}})
    ScaleA=true
    ScaleRGB=true
   )
   &ScaleSpriteSize(Scale Factor=&Vector2DFromFloat(Value=&FloatFromCurve(FloatCurve=DataInterface() {propertyValues {Curve {keys [<at 0 0 RCIM_Cubic>,<at 0.2 1 RCIM_Cubic>]}}})))
   &DynamicMaterialParameters(
    Index 0 Param 1=2
    Module.Param0WriteEnabled=true
    Module.Param1WriteEnabled=true
    Module.Param2WriteEnabled=true
    Module.Param3WriteEnabled=true
   )
   &CameraOffset(Camera Offset Amount=-200)
  ]
  rendererClasses [&NiagaraSpriteRendererProperties]
  renderers [{rendererClass=&NiagaraSpriteRendererProperties}]
 }
 Corona {
  emitterUpdateScript [
   &EmitterState(Life Cycle Mode=::NewEnumerator1, Loop Behavior=::NewEnumerator1, Loop Duration=10, UseLoopDelay=false)
   &SpawnBurst_Instantaneous(Spawn Count=1, Spawn Time=0.65, Use Spawn Probability=false)
  ]
  particleSpawnScript [&InitializeParticle(
    Color=<rgba 0.8317 0 2 0.2>
    Lifetime=0.4
    Mesh Scale=<1.5 1.5 1.5>
    Module.Write Color=true
    Module.Write Lifetime=true
    Module.Write Mass=true
    Module.Write Material Random=true
    Module.Write Position=false
    Module.Write Scale=false
    Module.Write SpriteRotation=true
    Module.Write SpriteSize=true
    Position=&SimulationPosition()
    Sprite Rotation=&UniformRangedFloat(Minimum=0, Maximum=1)
    Sprite Size=<270 270>
   )]
  particleUpdateScript [
   &ParticleState()
   &ScaleSpriteSize(Scale Factor=&Vector2DFromFloat(Value=&FloatFromCurve(
      FloatCurve=DataInterface() {propertyValues {Curve {keys [
          {arriveTangent=3.8601, interpMode=RCIM_Cubic, leaveTangent=3.8601, tangentMode=RCTM_User, time=0, value=0.5}
          {arriveTangent=2.25, interpMode=RCIM_Cubic, leaveTangent=2.25, time=0.2, value=1}
          <at 0.4 1.4>
         ]}}}
      CurveIndex=$Particles.Age
     )))
   &CameraOffset(Camera Offset Amount=20)
   &DynamicMaterialParameters(
    Index 0 Param 1=&FloatFromCurve(FloatCurve=DataInterface() {propertyValues {Curve {keys [<at -0.0046 1.0113>,<at 1 0.5>]}}})
    Index 0 Param 2=1
    Index 0 Param 3=1
    Index 0 Param 4=1
    Module.Param0WriteEnabled=true
    Module.Param1WriteEnabled=true
    Module.Param2WriteEnabled=true
    Module.Param3WriteEnabled=true
   )
  ]
  rendererClasses [&NiagaraSpriteRendererProperties]
  renderers [{rendererClass=&NiagaraSpriteRendererProperties}]
 }
 LightningMesh_Burst {
  emitterUpdateScript [
   &EmitterState(
    Inactive Response=::NewEnumerator1
    Life Cycle Mode=::NewEnumerator1
    Loop Behavior=::NewEnumerator1
    Loop Delay=2
    Loop Duration=10
    UseLoopDelay=false
   )
   /SetVariables(Emitter.RandSpawnGroup:NiagaraInt32=&MakeIntFromFloat(Float=&MakeFloatFromInt(Int=&UniformRangedInt(Minimum=0, Maximum=4)))) {isSetParameters=true}
   &SpawnBurst_Instantaneous(Spawn Count=6, Spawn Probability=0.5, Spawn Time=0.71, Use Spawn Probability=false)
  ]
  particleSpawnScript [
   &InitializeParticle(
    Color=<rgba 4.1587 0 10 1>
    Lifetime=&UniformRangedFloat(Minimum=0.2, Maximum=0.25)
    Mesh Scale=&UniformRangedVector(Minimum=<3 2 2>, Maximum=<3 3 3>)
    Module.Write Color=true
    Module.Write Lifetime=true
    Module.Write Mass=true
    Module.Write Material Random=true
    Module.Write Position=false
    Module.Write Scale=true
    Module.Write SpriteRotation=false
    Module.Write SpriteSize=false
    Position=&SimulationPosition()
   )
   &FN_SetSpawnGroup_v2(Number of Spawn Groups=5)
   &DynamicMaterialParameters(
    Index 0 Param 1=6
    Index 0 Param 3=&UniformRangedFloat(Minimum=1.45, Maximum=2.2)
    Index 0 Param 4=&UniformRangedFloat(Minimum=-60, Maximum=60)
    Module.Param0WriteEnabled=true
    Module.Param1WriteEnabled=false
    Module.Param2WriteEnabled=true
    Module.Param3WriteEnabled=true
   )
   /Niagara/Modules/Spawn/Orientation/V2/InitialMeshOrientation(
    Rotation=&UniformRangedVector(Minimum=<-1 0 0>, Maximum=<1 0 0>)
    Rotation Coordinate Space:ENiagaraMeshTransforms=::NewEnumerator1
    Use Orientation Vector=true
    Use Rotation Vector=true
   )
  ]
  particleUpdateScript [
   &ParticleState()
   &SkeletalMeshLocation(
    Bone / Socket Index=&SpawnGroup()
    Bone Sampling Mode=::NewEnumerator4
    Inherit Velocity=false
    MaskSpawnGroup=false
    Sampled Position Offset=<0 -40 0>
    Sampled Position Offset Coordinate Space:ENiagaraMeshTransforms=::NewEnumerator3
    Skeletal Mesh=$User.SkeletalMesh
   )
   /SetVariables(Particles.MeshOrientation=&MultiplyQuaternion(
     Quaternion A=&MultiplyQuaternion(
      Quaternion A=$Output.SkeletalMeshLocation.SampledRotation
      Quaternion B=&FindQuatBetweenTwoVectors(From Vector=<1 0 0>, To Vector=<0 1 0>)
     )
     Quaternion B=$Particles.Initial.MeshOrientation
    )) {isSetParameters=true}
   &DynamicMaterialParameters(
    Index 0 Param 2=&FloatFromCurve(FloatCurve=DataInterface() {propertyValues {Curve {keys [{arriveTangent=-2.9283, interpMode=RCIM_Cubic, leaveTangent=-2.9283, tangentMode=RCTM_User, time=0.1, value=1}, <at 1 -1>]}}})
    Module.Param0WriteEnabled=false
    Module.Param1WriteEnabled=true
    Module.Param2WriteEnabled=false
    Module.Param3WriteEnabled=false
   ) {module=DynamicMaterialParameters001}
   &ScaleColor(
    Scale RGB=&MakeVectorFromLinearColorRGB(Color=&ColorFromCurve(ColorCurve=DataInterface() {propertyValues {
        AlphaCurve {keys []}
        BlueCurve {keys [<at 0 2 RCIM_Cubic>,<at 1 4 RCIM_Cubic>]}
        GreenCurve {keys [<at 0 1.8653 RCIM_Cubic>,<at 1 2.6922 RCIM_Cubic>]}
        RedCurve {keys [<at 0 1.5312 RCIM_Cubic>,<at 1 0.8958 RCIM_Cubic>]}
       }}))
    ScaleA=true
    ScaleRGB=true
   )
   &CameraOffset(Camera Offset Amount=10)
   &FN_Light_v3(
    Color=<rgba 4.1587 0 10 1>
    Create Color Attribute=true
    Create Exponent Attribute=true
    Create Position Attribute=true
    Create Radius Attribute=true
    Enable Particle Light Enable=true
    Position:Vector3f=&Add_Vector(A=$Output.SkeletalMeshLocation.SampledPosition, B=<0 0 20>)
    Radius=200
   )
   /SetVariables(Particles.spritescale:Vector2f=<20 20>) {isSetParameters=true}
  ]
  rendererClasses [&NiagaraMeshRendererProperties, &NiagaraLightRendererProperties]
  renderers [
   {rendererClass=&NiagaraMeshRendererProperties, rendererIndex=0}
   {rendererClass=&NiagaraLightRendererProperties, rendererIndex=1}
  ]
 }
 LightningSprites_Burst {
  emitterUpdateScript [
   &EmitterState(
    Inactive Response=::NewEnumerator1
    Life Cycle Mode=::NewEnumerator1
    Loop Behavior=::NewEnumerator1
    Loop Delay=0.5
    Loop Duration=0.5
    UseLoopDelay=true
   )
   &SpawnRate(Spawn Probability=0.5, SpawnRate=10, Use Spawn Probability=false)
  ]
  particleSpawnScript [
   &InitializeParticle(
    Color=<rgba 0.8317 0 2 1>
    Module.Write Color=true
    Module.Write Lifetime=true
    Module.Write Mass=true
    Module.Write Material Random=true
    Module.Write Position=false
    Module.Write Scale=false
    Module.Write SpriteRotation=false
    Module.Write SpriteSize=true
    Position=&SimulationPosition()
    Sprite Rotation=&UniformRangedFloat(Minimum=0, Maximum=360)
    Sprite Size=&UniformRangedVector2D(Minimum=<40 40>, Maximum=<50 50>)
   )
   /Niagara/Modules/Spawn/Location/SphereLocation_System(
    Module.Override Randomness=false
    Non Uniform Scale:Vector3f=<1 1 0>
    Sphere Radius=200
    Surface Expansion Mode:ENiagaraExpansionMode=::NewEnumerator1
    Surface Only=true
   )
   /SetVariables(Particles.RandomPosition:Vector3f=&UniformRangedVector(Minimum=<-30 -30 0>, Maximum=<30 30 0>)) {isSetParameters=true, module=SetVariables_#1}
   &DynamicMaterialParameters(
    Index 0 Param 3=&FloatFromCurve(FloatCurve=DataInterface() {propertyValues {Curve {keys [<at 0 1 RCIM_Cubic>,<at 0.2 0.025 RCIM_Cubic>,<at 1 1>]}}})
    Module.Param0WriteEnabled=false
    Module.Param1WriteEnabled=false
    Module.Param2WriteEnabled=true
    Module.Param3WriteEnabled=false
   )
   &FN_SetSpawnGroup_v2(Number of Spawn Groups=5)
   /SetVariables(Particles.AirGlow_Color:LinearColor=<rgba 0.8317 0 2 1>, Particles.AirGlow_Size:Vector2f=<50 50>) {isSetParameters=true, module=SetVariables_#2}
  ]
  particleUpdateScript [
   &ParticleState()
   &SkeletalMeshLocation(
    Bone / Socket Index=&SpawnGroup()
    Bone Sampling Mode=::NewEnumerator4
    Inherit Velocity=false
    MaskSpawnGroup=false
    Sampled Position Offset=$Particles.RandomPosition
    Sampled Position Offset Coordinate Space:ENiagaraMeshTransforms=::NewEnumerator1
    Skeletal Mesh=$User.SkeletalMesh
   )
   /SetVariables(Particles.AirGlow_Position:NiagaraPosition=$Particles.Position) {isSetParameters=true}
   /SetVariables(Particles.SpriteFacing=&RotateVectorByQuaternion(Vector To Rotate=<1 0 0>, Quaternion=$Output.SkeletalMeshLocation.SampledRotation)) {isSetParameters=true, module=SetVariables_#3}
   &DynamicMaterialParameters(
    Index 0 Param 3=&FloatFromCurve(FloatCurve=DataInterface() {propertyValues {Curve {keys [<at 0 1 RCIM_Cubic>,<at 0.2 0.025 RCIM_Cubic>,<at 1 1>]}}})
    Module.Param0WriteEnabled=false
    Module.Param1WriteEnabled=false
    Module.Param2WriteEnabled=true
    Module.Param3WriteEnabled=false
   ) {module=DynamicMaterialParameters001}
   &ScaleColor(
    Scale Alpha=&FloatFromCurve(FloatCurve=DataInterface() {propertyValues {Curve {keys [<at 0.75 1 RCIM_Cubic>, {arriveTangent=-10.3917, interpMode=RCIM_Cubic, leaveTangent=-10.3917, tangentMode=RCTM_User, time=1, value=0}]}}})
    Scale RGB=&Multiply_VectorByFloat(
     Vector=&MakeVectorFromLinearColorRGB(Color=&ColorFromCurve(ColorCurve=DataInterface() {propertyValues {
         AlphaCurve {keys []}
         BlueCurve {keys [<at 0 1>,<at 0.15 10 RCIM_Cubic>,<at 1 1 RCIM_Cubic>]}
         GreenCurve {keys [<at 0 1>, {arriveTangent=-0.1823, interpMode=RCIM_Cubic, leaveTangent=-0.1823, time=0.15, value=9.045}, <at 1 0.8177 RCIM_Cubic>]}
         RedCurve {keys [<at 0 1>, {arriveTangent=-0.1812, interpMode=RCIM_Cubic, leaveTangent=-0.1812, time=0.15, value=8.5417}, <at 1 0.8188 RCIM_Cubic>]}
        }}))
     Float=1
    )
    ScaleA=true
    ScaleRGB=true
   )
   /Niagara/Modules/Update/SubUV/SubUVAnimation(Number Of Frames=8, Random SubUV Lookup Index:NiagaraFloat=$Particles.NormalizedAge, SubUV Animation Mode=::NewEnumerator1)
  ]
  rendererClasses [&NiagaraSpriteRendererProperties]
  renderers [
   {rendererClass=&NiagaraSpriteRendererProperties}
   {rendererClass=&NiagaraSpriteRendererProperties, rendererIndex=1}
  ]
 }
 Lightning_Sprites_Burst {
  emitterUpdateScript [
   &EmitterState(
    Inactive Response=::NewEnumerator1
    Life Cycle Mode=::NewEnumerator1
    Loop Behavior=::NewEnumerator1
    Loop Delay=2.4
    Loop Duration=0.5
    UseLoopDelay=true
   )
   &SpawnRate(Spawn Probability=0.5, SpawnRate=&UniformRangedFloat(Minimum=40, Maximum=40), Use Spawn Probability=false)
  ]
  particleSpawnScript [
   &InitializeParticle(
    Color=<rgba 0.8317 0 2 1>
    Module.Write Color=true
    Module.Write Lifetime=true
    Module.Write Mass=true
    Module.Write Material Random=true
    Module.Write Position=false
    Module.Write Scale=false
    Module.Write SpriteRotation=false
    Module.Write SpriteSize=true
    Position=&SimulationPosition()
    Sprite Rotation=&UniformRangedFloat(Minimum=0, Maximum=360)
    Sprite Size=&UniformRangedVector2D(Minimum=<40 40>, Maximum=<50 50>)
   )
   &SphereLocation_System(
    Module.Override Randomness=false
    Non Uniform Scale:Vector3f=<1 1 0>
    Sphere Radius=200
    Surface Expansion Mode:ENiagaraExpansionMode=::NewEnumerator1
    Surface Only=true
   )
   /SetVariables(Particles.RandomPosition:Vector3f=&UniformRangedVector(Minimum=<-30 -30 0>, Maximum=<30 30 0>)) {isSetParameters=true, module=SetVariables_#1}
   &DynamicMaterialParameters(
    Index 0 Param 3=&FloatFromCurve(FloatCurve=DataInterface() {propertyValues {Curve {keys [<at 0 1 RCIM_Cubic>,<at 0.2 0.025 RCIM_Cubic>,<at 1 1>]}}})
    Module.Param0WriteEnabled=false
    Module.Param1WriteEnabled=false
    Module.Param2WriteEnabled=true
    Module.Param3WriteEnabled=false
   )
   &FN_SetSpawnGroup_v2(Number of Spawn Groups=5)
   /SetVariables(Particles.AirGlow_Color:LinearColor=<rgba 0.8317 0 2 1>, Particles.AirGlow_Size:Vector2f=<50 50>) {isSetParameters=true, module=SetVariables_#2}
  ]
  particleUpdateScript [
   &ParticleState()
   &SkeletalMeshLocation(
    Bone / Socket Index=&SpawnGroup()
    Bone Sampling Mode=::NewEnumerator4
    Inherit Velocity=false
    MaskSpawnGroup=false
    Sampled Position Offset=$Particles.RandomPosition
    Sampled Position Offset Coordinate Space:ENiagaraMeshTransforms=::NewEnumerator1
    Skeletal Mesh=$User.SkeletalMesh
   )
   /SetVariables(Particles.SpriteFacing=&RotateVectorByQuaternion(Vector To Rotate=<1 0 0>, Quaternion=$Output.SkeletalMeshLocation.SampledRotation)) {isSetParameters=true, module=SetVariables_#3}
   &DynamicMaterialParameters(
    Index 0 Param 3=&FloatFromCurve(FloatCurve=DataInterface() {propertyValues {Curve {keys [<at 0 1 RCIM_Cubic>,<at 0.2 0.025 RCIM_Cubic>,<at 1 1>]}}})
    Module.Param0WriteEnabled=false
    Module.Param1WriteEnabled=false
    Module.Param2WriteEnabled=true
    Module.Param3WriteEnabled=false
   ) {module=DynamicMaterialParameters001}
   &ScaleColor(
    Scale Alpha=&FloatFromCurve(FloatCurve=DataInterface() {propertyValues {Curve {keys [<at 0.75 1 RCIM_Cubic>, {arriveTangent=-10.3917, interpMode=RCIM_Cubic, leaveTangent=-10.3917, tangentMode=RCTM_User, time=1, value=0}]}}})
    Scale RGB=&Multiply_VectorByFloat(
     Vector=&MakeVectorFromLinearColorRGB(Color=&ColorFromCurve(ColorCurve=DataInterface() {propertyValues {
         AlphaCurve {keys []}
         BlueCurve {keys [<at 0 1>,<at 0.15 10 RCIM_Cubic>,<at 1 1 RCIM_Cubic>]}
         GreenCurve {keys [<at 0 1>, {arriveTangent=-0.1823, interpMode=RCIM_Cubic, leaveTangent=-0.1823, time=0.15, value=9.045}, <at 1 0.8177 RCIM_Cubic>]}
         RedCurve {keys [<at 0 1>, {arriveTangent=-0.1812, interpMode=RCIM_Cubic, leaveTangent=-0.1812, time=0.15, value=8.5417}, <at 1 0.8188 RCIM_Cubic>]}
        }}))
     Float=1
    )
    ScaleA=true
    ScaleRGB=true
   )
   &SubUVAnimation(Number Of Frames=8, Random SubUV Lookup Index:NiagaraFloat=$Particles.NormalizedAge, SubUV Animation Mode=::NewEnumerator1)
  ]
  rendererClasses [&NiagaraSpriteRendererProperties]
  renderers [
   {rendererClass=&NiagaraSpriteRendererProperties}
   {rendererClass=&NiagaraSpriteRendererProperties, rendererIndex=1}
  ]
 }
 Lightning_Sprites_Loop {
  emitterUpdateScript [
   &EmitterState(
    Inactive Response=::NewEnumerator1
    Life Cycle Mode=::NewEnumerator1
    Loop Behavior=::NewEnumerator1
    Loop Delay=1
    Loop Duration=1
    UseLoopDelay=true
   )
   &SpawnRate(SpawnRate=5, Use Spawn Probability=false)
  ]
  particleSpawnScript [
   &InitializeParticle(
    Color=<rgba 0.8317 0 2 1>
    Mesh Scale=<50 50 15>
    Module.Write Color=true
    Module.Write Lifetime=true
    Module.Write Mass=true
    Module.Write Material Random=true
    Module.Write Position=false
    Module.Write Scale=true
    Module.Write SpriteRotation=false
    Module.Write SpriteSize=true
    Position=&SimulationPosition()
    Sprite Rotation=&UniformRangedFloat(Minimum=0, Maximum=360)
    Sprite Size=&UniformRangedVector2D(Minimum=<40 10>, Maximum=<50 50>)
   )
   /SetVariables(Particles.RandomPosition:Vector3f=&UniformRangedVector(Minimum=<0 -30 0>, Maximum=<0 30 0>)) {isSetParameters=true, module=SetVariables_#1}
   &DynamicMaterialParameters(
    Index 0 Param 3=&FloatFromCurve(FloatCurve=DataInterface() {propertyValues {Curve {keys [<at 0 1 RCIM_Cubic>,<at 0.2 0.025 RCIM_Cubic>,<at 1 1>]}}})
    Module.Param0WriteEnabled=false
    Module.Param1WriteEnabled=false
    Module.Param2WriteEnabled=true
    Module.Param3WriteEnabled=false
   )
   &FN_SetSpawnGroup_v2(Number of Spawn Groups=5)
   /SetVariables(Particles.AirGlow_Color:LinearColor=<rgba 4.1587 0 10 1>, Particles.AirGlow_Size:Vector2f=<150 150>) {isSetParameters=true}
   /SetVariables(Particles.SubImageIndex=1) {isSetParameters=true}
   /SetVariables(Particles.RandomRot:Quat4f=&MakeQuatFromAxisAngle(Axis=<0 0 1>, AngleInDegrees=&UniformRangedFloat(Minimum=-45, Maximum=45))) {isSetParameters=true}
   /SetVariables(Particles.MeshOrientation=&MultiplyQuaternion(
     Quaternion A=$Particles.RandomRot
     Quaternion B=&FindQuatBetweenTwoVectors(From Vector=<0 1 0>, To Vector=<0 0 -1>)
    )) {isSetParameters=true}
  ]
  particleUpdateScript [
   &ParticleState()
   &SkeletalMeshLocation(
    Bone / Socket Index=$Particles.SpawnGroup
    Bone Sampling Mode=::NewEnumerator4
    Inherit Velocity=false
    MaskSpawnGroup=false
    Sampled Position Offset=$Particles.RandomPosition
    Sampled Position Offset Coordinate Space:ENiagaraMeshTransforms=::NewEnumerator3
    Skeletal Mesh=$User.SkeletalMesh
   )
   /SetVariables(
    Particles.Decal_CameraOffset=5
    Particles.Decal_Color:LinearColor=&ColorFromCurve(ColorCurve=DataInterface() {propertyValues {
       AlphaCurve {keys [<at 0 0 RCIM_Cubic>,<at 0.05 0.5>,<at 0.25 0.5>,<at 0.5 0 RCIM_Cubic>]}
       BlueCurve {keys [<at 0 2 RCIM_Cubic>,<at 0.25 1>]}
       GreenCurve {keys [<at 0 0 RCIM_Cubic>,<at 0.25 0>]}
       RedCurve {keys [<at 0 1.5807 RCIM_Cubic>,<at 0.25 0.7904>]}
      }})
    Particles.Decal_DynamicParameter:Vector4f=&MakeVector4(X=1, Y=1, Z=1, W=1)
    Particles.Decal_LT_Position:Vector3f=&Add_Vector(
     A=Linked() {linkedVariable {name=Particles.Position, type=NiagaraPosition}}
     B=&RotateVectorByQuaternion(Vector To Rotate=<-15 0 0>, Quaternion=$Particles.RandomRot)
    )
    Particles.Decal_RT_Position:Vector3f=&Add_Vector(
     A=Linked() {linkedVariable {name=Particles.Position, type=NiagaraPosition}}
     B=&RotateVectorByQuaternion(Vector To Rotate=<12 0 0>, Quaternion=$Particles.RandomRot)
    )
    Particles.Decal_Size:Vector3f=<20 20 0>
   ) {isSetParameters=true}
   /SetVariables(Particles.SpriteFacing=&RotateVectorByQuaternion(Vector To Rotate=<1 0 0>, Quaternion=$Output.SkeletalMeshLocation.SampledRotation)) {isSetParameters=true, module=SetVariables_#3}
   &DynamicMaterialParameters(
    Index 0 Param 3=&FloatFromCurve(FloatCurve=DataInterface() {propertyValues {Curve {keys [<at 0 1 RCIM_Cubic>,<at 0.05 0 RCIM_Cubic>,<at 1 1 RCIM_Cubic>]}}})
    Module.Param0WriteEnabled=false
    Module.Param1WriteEnabled=false
    Module.Param2WriteEnabled=true
    Module.Param3WriteEnabled=false
   ) {module=DynamicMaterialParameters001}
   &ScaleColor(
    Scale Alpha=&FloatFromCurve(FloatCurve=DataInterface() {propertyValues {Curve {keys [<at 0.75 1 RCIM_Cubic>, {arriveTangent=-10.3917, interpMode=RCIM_Cubic, leaveTangent=-10.3917, tangentMode=RCTM_User, time=1, value=0}]}}})
    Scale RGB=&Multiply_VectorByFloat(
     Vector=&MakeVectorFromLinearColorRGB(Color=&ColorFromCurve(ColorCurve=DataInterface() {propertyValues {
         AlphaCurve {keys []}
         BlueCurve {keys [<at 0 1>,<at 0.15 10 RCIM_Cubic>,<at 1 1 RCIM_Cubic>]}
         GreenCurve {keys [<at 0 1>, {arriveTangent=-0.1823, interpMode=RCIM_Cubic, time=0.15, value=9.045}, <at 1 0.8177 RCIM_Cubic>]}
         RedCurve {keys [<at 0 1>, {arriveTangent=-0.1812, interpMode=RCIM_Cubic, time=0.15, value=8.5417}, <at 1 0.8188 RCIM_Cubic>]}
        }}))
     Float=1
    )
    ScaleA=true
    ScaleRGB=true
   )
   &CameraOffset(Camera Offset Amount=-5)
  ]
  rendererClasses [&NiagaraSpriteRendererProperties, &NiagaraMeshRendererProperties]
  renderers [
   {rendererClass=&NiagaraSpriteRendererProperties}
   {rendererClass=&NiagaraMeshRendererProperties, rendererIndex=1}
   {rendererClass=&NiagaraMeshRendererProperties, rendererIndex=2}
   {rendererClass=&NiagaraMeshRendererProperties, rendererIndex=3}
  ]
 }
}
summary {
 systemName=NS_KevinCouture_Glider_Deploy
 emitters [
  {
   emitterName=LightningMesh_Burst
   rendererClasses [&NiagaraMeshRendererProperties, &NiagaraLightRendererProperties]
  }
  {emitterName=Lightning_Sprites_Burst, rendererClasses [&NiagaraSpriteRendererProperties]}
  {emitterName=LightningSprites_Burst, rendererClasses [&NiagaraSpriteRendererProperties]}
  {emitterName=AirGlow, rendererClasses [&NiagaraSpriteRendererProperties]}
  {emitterName=Corona, rendererClasses [&NiagaraSpriteRendererProperties]}
  {
   emitterName=Lightning_Sprites_Loop
   rendererClasses [&NiagaraSpriteRendererProperties, &NiagaraMeshRendererProperties]
  }
 ]
 userVariables [User.SkeletalMesh {
   type=NiagaraDataInterfaceSkeletalMesh
   defaultValue {
    struct=NiagaraExt_VariableValue_DataInterface
    value {dataInterface=&self:NiagaraDataInterfaceSkeletalMesh_0}
   }
  }]
}
