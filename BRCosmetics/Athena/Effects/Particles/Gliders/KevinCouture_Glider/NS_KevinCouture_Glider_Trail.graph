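# Niagara trail and aura effect for the KevinCouture glider cosmetic, layering mesh trails, purple lightning, runes, decals and lights.
# Emitters sample bones and sockets on two user skeletal meshes to place lightning, rune sprites and glow decals, with dynamic material parameters animating the arcs.
# The mesh trail bends with exhaust vector and scales by deploy state, opacity and TrailLengthScale user inputs, using front-end modifiers to switch between lobby preview and gameplay values.
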NiagaraSystem /BRCosmetics/Athena/Effects/Particles/Gliders/KevinCouture_Glider/NS_KevinCouture_Glider_Trail
dependencies {
 usedDataInterfaces [&NiagaraDataInterfaceCurve, &NiagaraDataInterfaceColorCurve]
 usedDynamicInputs []
 usedModules [&SystemState]
}
emitters {
 BurstRunes {
  emitterUpdateScript [
   &EmitterState()
   &SpawnRate(SpawnRate=10, Use Spawn Probability=false)
  ]
  particleSpawnScript [
   /Niagara/Modules/Spawn/Initialization/InitializeParticle(
    Color=&MakeLinearColor(
     R:NiagaraFloat=1
     G:NiagaraFloat=&UniformRangedFloat(Minimum=5, Maximum=10)
     B:NiagaraFloat=&UniformRangedFloat(Minimum=1, Maximum=10)
     A:NiagaraFloat=&UniformRangedFloat(Minimum=0.5, Maximum=1)
    )
    Lifetime=&UniformRangedFloat(Minimum=0.2, Maximum=0.5)
    Module.Write Color=true
    Module.Write Lifetime=true
    Module.Write Mass=false
    Module.Write Material Random=true
    Module.Write Position=false
    Module.Write Scale=false
    Module.Write SpriteRotation=true
    Module.Write SpriteSize=true
    Position=&SimulationPosition()
    Sprite Rotation=&UniformRangedFloat(Minimum=0, Maximum=360)
    Sprite Size=&Multiply_Vector2(
     A=&Vector2DFromFloat(Value=&UniformRangedFloat(Minimum=10, Maximum=16.5))
     B=&FN_FrontEndVector2Modifier(
      Front End:Vector2f=&Vector2DFromFloat(Value=0.65)
      Gameplay:Vector2f=&Vector2DFromFloat(Value=$User.TrailLengthScale)
     )
    )
   )
   /SetVariables(Particles.SpriteInitRotationRate:NiagaraFloat=&Multiply_Float(
     A=&UniformAOrBFloat(A=-1)
     B=&UniformRangedFloat(Minimum=5, Maximum=10)
    )) {isSetParameters=true}
   /Niagara/Modules/Update/SubUV/SubUVAnimation(SubUV Animation Mode=::NewEnumerator1)
   &SkeletalMeshLocation(
    Bone / Socket Index=2
    Inherit Velocity=false
    MaskSpawnGroup=false
    Mesh Sampling Type=::NewEnumerator5
    Sampled Position Offset=&UniformRangedVector(Minimum=<0 -40 0>, Maximum=<0 40 0>)
    Sampled Position Offset Coordinate Space:ENiagaraMeshTransforms=::NewEnumerator3
    Skeletal Mesh=$User.SkeletalMesh
    Socket Sampling Mode=::NewEnumerator5
   )
   &AddVelocity(Velocity=&Multiply_Vector(
     A=&FN_FrontEndVector3Modifier(Gameplay=<0 200 -75>, Front End=<0 200 -75>)
     B=&MakeVector(X=0, Y=&FN_FrontEndFloatModifier(Gameplay=$User.TrailLengthScale, Front End=1), Z=1)
    ))
  ]
  particleUpdateScript [
   &ParticleState()
   &SolveForcesAndVelocity(Clamp Velocity=false, Limit Acceleration=false)
   &ScaleColor(
    Scale Alpha=&FloatFromCurve(FloatCurve=DataInterface() {propertyValues {Curve {keys [{interpMode=RCIM_Cubic, tangentMode=RCTM_User, time=0.75, value=1}, <at 1 0.5>]}}})
    Scale RGB=&MakeVector(
     X=&UniformAOrBFloat(A=0)
     Y=1
     Z=&FloatFromCurve(FloatCurve=DataInterface() {propertyValues {Curve {keys [
          {arriveTangent=-172.0061, interpMode=RCIM_Cubic, leaveTangent=-172.0061, tangentMode=RCTM_User, time=0.2, value=10}
          {arriveTangent=-0.0507, interpMode=RCIM_Cubic, leaveTangent=-0.0508, tangentMode=RCTM_User, time=0.32, value=1}
          {arriveTangent=90.7103, interpMode=RCIM_Cubic, leaveTangent=90.7105, tangentMode=RCTM_User, time=1, value=20}
         ]}}})
    )
    ScaleA=true
    ScaleRGB=true
   )
  ]
  rendererClasses [&NiagaraSpriteRendererProperties]
  renderers [{rendererClass=&NiagaraSpriteRendererProperties}]
 }
 Decal {
  emitterUpdateScript [
   &EmitterState(
    Inactive Response=::NewEnumerator1
    Life Cycle Mode=::NewEnumerator1
    Loop Behavior=::NewEnumerator1
    Loop Duration Mode=::NewEnumerator1
    UseLoopDelay=false
   )
   &SpawnBurst_Instantaneous(Spawn Count=1, Use Spawn Probability=false)
  ]
  particleSpawnScript [&InitializeParticle(
    Color=<rgba 2 2 2 1>
    Lifetime=5
    Mesh Scale=&FN_FrontEndVector3Modifier(Gameplay=<80 10 1>, Front End=<52 6 1>)
    Module.Write Color=true
    Module.Write Lifetime=true
    Module.Write Mass=true
    Module.Write Material Random=true
    Module.Write Position=false
    Module.Write Scale=true
    Module.Write SpriteRotation=false
    Module.Write SpriteSize=false
    Position=&SimulationPosition()
    Sprite Size=<100 20>
   )]
  particleUpdateScript [
   &ParticleState(Kill Particles When Lifetime Has Elapsed=false)
   &SkeletalMeshLocation(
    Bone / Socket Index=2
    Inherit Velocity=false
    MaskSpawnGroup=false
    Mesh Sampling Type=::NewEnumerator5
    Skeletal Mesh=$User.SkeletalMesh
    Socket Sampling Mode=::NewEnumerator5
   )
   &DynamicMaterialParameters(Module.Param0WriteEnabled=true, Module.Param1WriteEnabled=true, Module.Param2WriteEnabled=true, Module.Param3WriteEnabled=true)
   /SetVariables(Particles.MeshOrientation=&MultiplyQuaternion(
     Quaternion A=$Output.SkeletalMeshLocation.SampledRotation
     Quaternion B=&MultiplyQuaternion(
      Quaternion A=&FindQuatBetweenTwoVectors(From Vector=<1 0 0>, To Vector=<0 1 0>)
      Quaternion B=&FindQuatBetweenTwoVectors(From Vector=<0 1 0>, To Vector=<0 0 1>)
     )
    )) {isSetParameters=true}
   &CameraOffset(Camera Offset Amount=20)
  ]
  rendererClasses [&NiagaraMeshRendererProperties]
  renderers [{rendererClass=&NiagaraMeshRendererProperties, rendererIndex=0}]
 }
 Feet_Decal {
  emitterUpdateScript [
   &EmitterState(Life Cycle Mode=::NewEnumerator1, Loop Behavior=::NewEnumerator1, Loop Delay=2.4, Loop Duration Mode=::NewEnumerator1, UseLoopDelay=false)
   &SpawnBurst_Instantaneous(Spawn Count=2, Use Spawn Probability=false)
  ]
  particleSpawnScript [
   &InitializeParticle(
    Mesh Scale=<20 20 20>
    Module.Write Color=true
    Module.Write Lifetime=true
    Module.Write Mass=true
    Module.Write Material Random=true
    Module.Write Position=false
    Module.Write Scale=true
    Module.Write SpriteRotation=false
    Module.Write SpriteSize=true
    Position=&SimulationPosition()
    Sprite Size=<20 20>
   )
   &FN_SetSpawnGroup_v2(Number of Spawn Groups=2)
   /SetVariables(Particles.Position=&MaskVectorBySpawnGroup(Masked Vector=<7 30 0>, Passthrough Vector=<7 -40 0>)) {isSetParameters=true}
   /SetVariables(Particles.MeshOrientation=&MultiplyQuaternion(
     Quaternion A=$Particles.RandomRot
     Quaternion B=&FindQuatBetweenTwoVectors(From Vector=<0 1 0>, To Vector=<0 0 0>)
    )) {isSetParameters=true}
  ]
  particleUpdateScript [
   &ParticleState(Kill Particles When Lifetime Has Elapsed=false, Let Infinitely Lived Particles Die When Emitter Deactivates=true)
   /SetVariables(
    Particles.Decal_CameraOffset=5
    Particles.Decal_Color:LinearColor=&ColorFromCurve(ColorCurve=DataInterface() {propertyValues {
       AlphaCurve {keys [<at 0 0 RCIM_Cubic>,<at 0.05 0.2>,<at 0.25 0.2>]}
       BlueCurve {keys [<at 0 5 RCIM_Cubic>,<at 0.25 1>]}
       GreenCurve {keys [<at 0 0 RCIM_Cubic>,<at 0.25 0>]}
       RedCurve {keys [<at 0 3.9519 RCIM_Cubic>,<at 0.25 0.7904>]}
      }})
    Particles.Decal_DynamicParameter:Vector4f=&MakeVector4(X=0.5, Y=1, Z=1, W=1)
    Particles.Decal_LT_Position:NiagaraPosition=$Particles.Position
    Particles.Decal_RT_Position:NiagaraPosition=$Particles.Position
    Particles.Decal_Size:Vector3f=&VectorFromFloat(Value=&FloatFromCurve(FloatCurve=DataInterface() {propertyValues {Curve {keys [{arriveTangent=67.5571, interpMode=RCIM_Cubic, leaveTangent=67.5555, tangentMode=RCTM_User, time=0, value=50}, <at 1 75 RCIM_Cubic>]}}}))
   ) {isSetParameters=true}
  ]
  rendererClasses [&NiagaraMeshRendererProperties]
  renderers [{rendererClass=&NiagaraMeshRendererProperties, rendererIndex=0}]
 }
 Light {
  emitterUpdateScript [
   &EmitterState(Life Cycle Mode=::NewEnumerator1, Loop Duration=1, UseLoopDelay=false)
   &SpawnRate(SpawnRate=&UniformRangedFloat(Minimum=0, Maximum=5), Use Spawn Probability=false)
  ]
  particleSpawnScript [
   &InitializeParticle(
    Lifetime=0.2
    Module.Write Color=true
    Module.Write Lifetime=true
    Module.Write Mass=true
    Module.Write Material Random=true
    Module.Write Position=false
    Module.Write Scale=false
    Module.Write SpriteRotation=false
    Module.Write SpriteSize=true
    Position=&SimulationPosition()
    Sprite Size=<5 5>
   )
   &CylinderLocation(
    Cylinder Height=10
    Cylinder Radius=50
    MaskSpawnGroup=false
    Module.Enable Lathe Profile=false
    Module.Override Randomness=false
    Module.Surface Only=true
    Offset=<0 0 60>
    Surface Only Band Thickness=0
    Use Endcaps In Surface Only Mode=false
   )
  ]
  particleUpdateScript [
   &ParticleState()
   &FN_Light_v3(Color=<rgba 1.2476 0 3 1>, Create Color Attribute=true, Create Exponent Attribute=true, Create Radius Attribute=true, Radius=100)
  ]
  rendererClasses [&NiagaraLightRendererProperties]
  renderers [{rendererClass=&NiagaraLightRendererProperties, rendererIndex=0}]
 }
 Lightning_Mesh_Loop {
  emitterUpdateScript [
   &EmitterState()
   /SetVariables(Emitter.RandSpawnGroup:NiagaraInt32=&MakeIntFromFloat(Float=&MakeFloatFromInt(Int=&UniformRangedInt(Minimum=0, Maximum=4)))) {isSetParameters=true}
   &SpawnRate(SpawnRate=&UniformRangedFloat(Minimum=0, Maximum=5), Use Spawn Probability=false)
  ]
  particleSpawnScript [
   &InitializeParticle(
    Color=<rgba 0.8317 0 2 1>
    Lifetime=&UniformRangedFloat(Minimum=0.2, Maximum=0.25)
    Mesh Scale=&UniformRangedVector(Minimum=<3 2 2>, Maximum=<3 3 3>)
    Module.Write Color=true
    Module.Write Lifetime=true
    Module.Write Mass=true
    Module.Write Material Random=true
    Module.Write Position=false
    Module.Write Scale=true
    Module.Write SpriteRotation=false
    Module.Write SpriteSize=false
    Position=&SimulationPosition()
   )
   &FN_SetSpawnGroup_v2(Number of Spawn Groups=5)
   &DynamicMaterialParameters(
    Index 0 Param 1=6
    Index 0 Param 3=&UniformRangedFloat(Minimum=1.45, Maximum=2.2)
    Index 0 Param 4=&UniformRangedFloat(Minimum=-60, Maximum=60)
    Module.Param0WriteEnabled=true
    Module.Param1WriteEnabled=false
    Module.Param2WriteEnabled=true
    Module.Param3WriteEnabled=true
   )
   /Niagara/Modules/Spawn/Orientation/V2/InitialMeshOrientation(
    Rotation=&UniformRangedVector(Minimum=<-1 0 0>, Maximum=<1 0 0>)
    Rotation Coordinate Space:ENiagaraMeshTransforms=::NewEnumerator1
    Use Orientation Vector=true
    Use Rotation Vector=true
   )
   /SetVariables(Particles.AirGlow_Color:LinearColor=<rgba 0.8317 0 2 1>, Particles.AirGlow_Size:Vector2f=<50 50>) {isSetParameters=true}
  ]
  particleUpdateScript [
   &ParticleState()
   &SkeletalMeshLocation(
    Bone / Socket Index=$Particles.SpawnGroup
    Bone Sampling Mode=::NewEnumerator4
    Inherit Velocity=false
    MaskSpawnGroup=false
    Sampled Position Offset=<0 -40 0>
    Sampled Position Offset Coordinate Space:ENiagaraMeshTransforms=::NewEnumerator3
    Skeletal Mesh=$User.SkeletalMesh
   )
   /SetVariables(Particles.AirGlow_Position:Vector3f=$Output.SkeletalMeshLocation.SampledPosition) {isSetParameters=true}
   /SetVariables(Particles.MeshOrientation=&MultiplyQuaternion(
     Quaternion A=&MultiplyQuaternion(
      Quaternion A=$Output.SkeletalMeshLocation.SampledRotation
      Quaternion B=&FindQuatBetweenTwoVectors(From Vector=<1 0 0>, To Vector=<0 1 0>)
     )
     Quaternion B=$Particles.Initial.MeshOrientation
    )) {isSetParameters=true}
   /SetVariables(Particles.Position=&Add_Vector(A=Linked() {linkedVariable {name=Particles.Position, type=NiagaraPosition}}, B=<0 0 5>)) {isSetParameters=true}
   &DynamicMaterialParameters(
    Index 0 Param 2=&FloatFromCurve(FloatCurve=DataInterface() {propertyValues {Curve {keys [{arriveTangent=-2.9283, interpMode=RCIM_Cubic, leaveTangent=-2.9283, tangentMode=RCTM_User, time=0.1, value=1}, <at 1 -1>]}}})
    Module.Param0WriteEnabled=false
    Module.Param1WriteEnabled=true
    Module.Param2WriteEnabled=false
    Module.Param3WriteEnabled=false
   ) {module=DynamicMaterialParameters001}
   &ScaleColor(
    Scale RGB=&MakeVectorFromLinearColorRGB(Color=&ColorFromCurve(ColorCurve=DataInterface() {propertyValues {
        AlphaCurve {keys []}
        BlueCurve {keys [<at 0 2 RCIM_Cubic>,<at 1 4 RCIM_Cubic>]}
        GreenCurve {keys [<at 0 1.8653 RCIM_Cubic>,<at 1 2.6922 RCIM_Cubic>]}
        RedCurve {keys [<at 0 1.5312 RCIM_Cubic>,<at 1 0.8958 RCIM_Cubic>]}
       }}))
    ScaleA=true
    ScaleRGB=true
   )
   &CameraOffset(Camera Offset Amount=5)
   &FN_Light_v3(
    Color=<rgba 1.2476 0 3 1>
    Create Color Attribute=true
    Create Exponent Attribute=true
    Create Position Attribute=true
    Create Radius Attribute=true
    Enable Particle Light Enable=true
    Position:Vector3f=&Add_Vector(A=$Output.SkeletalMeshLocation.SampledPosition, B=<0 0 60>)
    Radius=60
   )
  ]
  rendererClasses [&NiagaraMeshRendererProperties, &NiagaraSpriteRendererProperties]
  renderers [
   {rendererClass=&NiagaraMeshRendererProperties, rendererIndex=0}
   {rendererClass=&NiagaraSpriteRendererProperties, rendererIndex=1}
  ]
 }
 Lightning_Sprites_Burst {
  emitterUpdateScript [
   &EmitterState(Life Cycle Mode=::NewEnumerator1, Loop Behavior=::NewEnumerator1, Loop Delay=2.4, Loop Duration=0.5, UseLoopDelay=false)
   &SpawnRate(
    Spawn Probability=0.5
    SpawnRate=&FN_FrontEndFloatModifier(Gameplay=0, Front End=20)
    Use Spawn Probability=false
   )
  ]
  particleSpawnScript [
   &InitializeParticle(
    Color=<rgba 0.8317 0 2 1>
    Module.Write Color=true
    Module.Write Lifetime=true
    Module.Write Mass=true
    Module.Write Material Random=true
    Module.Write Position=false
    Module.Write Scale=false
    Module.Write SpriteRotation=false
    Module.Write SpriteSize=true
    Position=&SimulationPosition()
    Sprite Rotation=&UniformRangedFloat(Minimum=0, Maximum=360)
    Sprite Size=&UniformRangedVector2D(Minimum=<40 40>, Maximum=<50 50>)
   )
   /Niagara/Modules/Spawn/Location/SphereLocation_System(
    Module.Override Randomness=false
    Non Uniform Scale:Vector3f=<1 1 0>
    Sphere Radius=200
    Surface Expansion Mode:ENiagaraExpansionMode=::NewEnumerator1
    Surface Only=true
   )
   /SetVariables(Particles.RandomPosition:Vector3f=&UniformRangedVector(Minimum=<-30 -30 0>, Maximum=<30 30 0>)) {isSetParameters=true, module=SetVariables_#1}
   &DynamicMaterialParameters(
    Index 0 Param 3=&FloatFromCurve(FloatCurve=DataInterface() {propertyValues {Curve {keys [<at 0 1 RCIM_Cubic>,<at 0.2 0.025 RCIM_Cubic>,<at 1 1>]}}})
    Module.Param0WriteEnabled=false
    Module.Param1WriteEnabled=false
    Module.Param2WriteEnabled=true
    Module.Param3WriteEnabled=false
   )
   &FN_SetSpawnGroup_v2(Number of Spawn Groups=5)
   /SetVariables(Particles.AirGlow_Color:LinearColor=<rgba 0.8317 0 2 1>, Particles.AirGlow_Size:Vector2f=<50 50>) {isSetParameters=true}
  ]
  particleUpdateScript [
   &ParticleState()
   &SkeletalMeshLocation(
    Bone / Socket Index=&SpawnGroup()
    Bone Sampling Mode=::NewEnumerator4
    Inherit Velocity=false
    MaskSpawnGroup=false
    Sampled Position Offset=$Particles.RandomPosition
    Sampled Position Offset Coordinate Space:ENiagaraMeshTransforms=::NewEnumerator1
    Skeletal Mesh=$User.SkeletalMesh
   )
   /SetVariables(Particles.SpriteFacing=&RotateVectorByQuaternion(Vector To Rotate=<1 0 0>, Quaternion=$Output.SkeletalMeshLocation.SampledRotation)) {isSetParameters=true, module=SetVariables_#2}
   &DynamicMaterialParameters(
    Index 0 Param 3=&FloatFromCurve(FloatCurve=DataInterface() {propertyValues {Curve {keys [<at 0 1 RCIM_Cubic>,<at 0.2 0.025 RCIM_Cubic>,<at 1 1>]}}})
    Module.Param0WriteEnabled=false
    Module.Param1WriteEnabled=false
    Module.Param2WriteEnabled=true
    Module.Param3WriteEnabled=false
   ) {module=DynamicMaterialParameters001}
   &ScaleColor(
    Scale Alpha=&FloatFromCurve(FloatCurve=DataInterface() {propertyValues {Curve {keys [<at 0.75 1 RCIM_Cubic>, {arriveTangent=-10.3917, interpMode=RCIM_Cubic, leaveTangent=-10.3917, tangentMode=RCTM_User, time=1, value=0}]}}})
    Scale RGB=&Multiply_VectorByFloat(
     Vector=&MakeVectorFromLinearColorRGB(Color=&ColorFromCurve(ColorCurve=DataInterface() {propertyValues {
         AlphaCurve {keys []}
         BlueCurve {keys [<at 0 1>,<at 0.15 10 RCIM_Cubic>,<at 1 1 RCIM_Cubic>]}
         GreenCurve {keys [<at 0 1>, {arriveTangent=-0.1823, interpMode=RCIM_Cubic, leaveTangent=-0.1823, time=0.15, value=9.045}, <at 1 0.8177 RCIM_Cubic>]}
         RedCurve {keys [<at 0 1>, {arriveTangent=-0.1812, interpMode=RCIM_Cubic, leaveTangent=-0.1812, time=0.15, value=8.5417}, <at 1 0.8188 RCIM_Cubic>]}
        }}))
     Float=1
    )
    ScaleA=true
    ScaleRGB=true
   )
   &SubUVAnimation(Number Of Frames=8, Random SubUV Lookup Index:NiagaraFloat=$Particles.NormalizedAge, SubUV Animation Mode=::NewEnumerator1)
  ]
  rendererClasses [&NiagaraSpriteRendererProperties]
  renderers [
   {rendererClass=&NiagaraSpriteRendererProperties}
   {rendererClass=&NiagaraSpriteRendererProperties, rendererIndex=1}
  ]
 }
 Lightning_Sprites_Loop {
  emitterUpdateScript [
   &EmitterState(
    Delay First Loop Only=true
    Inactive Response=::NewEnumerator1
    Life Cycle Mode=::NewEnumerator1
    Loop Delay=0.2
    Loop Duration=1
    Recalculate Duration Each Loop=true
    UseLoopDelay=true
   )
   &SpawnRate(Spawn Probability=0.5, SpawnRate=&UniformRangedFloat(Minimum=0, Maximum=10), Use Spawn Probability=true)
  ]
  particleSpawnScript [
   &InitializeParticle(
    Color=<rgba 0.8317 0 2 1>
    Mesh Scale=<50 50 15>
    Module.Write Color=true
    Module.Write Lifetime=true
    Module.Write Mass=true
    Module.Write Material Random=true
    Module.Write Position=false
    Module.Write Scale=true
    Module.Write SpriteRotation=false
    Module.Write SpriteSize=true
    Position=&SimulationPosition()
    Sprite Rotation=&UniformRangedFloat(Minimum=0, Maximum=360)
    Sprite Size=&UniformRangedVector2D(Minimum=<40 10>, Maximum=<50 50>)
   )
   /SetVariables(Particles.RandomPosition:Vector3f=&UniformRangedVector(Minimum=<0 -30 0>, Maximum=<0 30 0>)) {isSetParameters=true, module=SetVariables_#1}
   &DynamicMaterialParameters(
    Index 0 Param 3=&FloatFromCurve(FloatCurve=DataInterface() {propertyValues {Curve {keys [<at 0 1 RCIM_Cubic>,<at 0.2 0.025 RCIM_Cubic>,<at 1 1>]}}})
    Module.Param0WriteEnabled=false
    Module.Param1WriteEnabled=false
    Module.Param2WriteEnabled=true
    Module.Param3WriteEnabled=false
   )
   &FN_SetSpawnGroup_v2(Number of Spawn Groups=5)
   /SetVariables(Particles.AirGlow_Color:LinearColor=<rgba 4.1587 0 10 1>, Particles.AirGlow_Size:Vector2f=<150 150>) {isSetParameters=true, module=SetVariables_#3}
   /SetVariables(Particles.SubImageIndex=1) {isSetParameters=true, module=SetVariables_#4}
   /SetVariables(Particles.RandomRot:Quat4f=&MakeQuatFromAxisAngle(Axis=<0 0 1>, AngleInDegrees=&UniformRangedFloat(Minimum=-45, Maximum=45))) {isSetParameters=true, module=SetVariables_#5}
   /SetVariables(Particles.MeshOrientation=&MultiplyQuaternion(
     Quaternion A=$Particles.RandomRot
     Quaternion B=&FindQuatBetweenTwoVectors(From Vector=<0 1 0>, To Vector=<0 0 -1>)
    )) {isSetParameters=true, module=SetVariables_#6}
  ]
  particleUpdateScript [
   &ParticleState()
   &SkeletalMeshLocation(
    Bone / Socket Index=$Particles.SpawnGroup
    Bone Sampling Mode=::NewEnumerator4
    Inherit Velocity=false
    MaskSpawnGroup=false
    Sampled Position Offset=$Particles.RandomPosition
    Sampled Position Offset Coordinate Space:ENiagaraMeshTransforms=::NewEnumerator3
    Skeletal Mesh=$User.SkeletalMesh
   )
   /SetVariables(
    Particles.Decal_CameraOffset=5
    Particles.Decal_Color:LinearColor=&ColorFromCurve(ColorCurve=DataInterface() {propertyValues {
       AlphaCurve {keys [<at 0 0 RCIM_Cubic>,<at 0.05 0.5>,<at 0.25 0.5>,<at 0.5 0 RCIM_Cubic>]}
       BlueCurve {keys [<at 0 2 RCIM_Cubic>,<at 0.25 1>]}
       GreenCurve {keys [<at 0 0 RCIM_Cubic>,<at 0.25 0>]}
       RedCurve {keys [<at 0 1.5807 RCIM_Cubic>,<at 0.25 0.7904>]}
      }})
    Particles.Decal_DynamicParameter:Vector4f=&MakeVector4(X=1, Y=1, Z=1, W=1)
    Particles.Decal_LT_Position:Vector3f=&Add_Vector(
     A=Linked() {linkedVariable {name=Particles.Position, type=NiagaraPosition}}
     B=&RotateVectorByQuaternion(Vector To Rotate=<-15 0 0>, Quaternion=$Particles.RandomRot)
    )
    Particles.Decal_RT_Position:Vector3f=&Add_Vector(
     A=Linked() {linkedVariable {name=Particles.Position, type=NiagaraPosition}}
     B=&RotateVectorByQuaternion(Vector To Rotate=<12 0 0>, Quaternion=$Particles.RandomRot)
    )
    Particles.Decal_Size:Vector3f=<20 20 0>
   ) {isSetParameters=true, module=SetVariables_#7}
   /SetVariables(Particles.SpriteFacing=&RotateVectorByQuaternion(Vector To Rotate=<1 0 0>, Quaternion=$Output.SkeletalMeshLocation.SampledRotation)) {isSetParameters=true, module=SetVariables_#2}
   &DynamicMaterialParameters(
    Index 0 Param 3=&FloatFromCurve(FloatCurve=DataInterface() {propertyValues {Curve {keys [<at 0 1 RCIM_Cubic>,<at 0.05 0 RCIM_Cubic>,<at 1 1 RCIM_Cubic>]}}})
    Module.Param0WriteEnabled=false
    Module.Param1WriteEnabled=false
    Module.Param2WriteEnabled=true
    Module.Param3WriteEnabled=false
   ) {module=DynamicMaterialParameters001}
   &ScaleColor(
    Scale Alpha=&FloatFromCurve(FloatCurve=DataInterface() {propertyValues {Curve {keys [<at 0.75 1 RCIM_Cubic>, {arriveTangent=-10.3917, interpMode=RCIM_Cubic, leaveTangent=-10.3917, tangentMode=RCTM_User, time=1, value=0}]}}})
    Scale RGB=&Multiply_VectorByFloat(
     Vector=&MakeVectorFromLinearColorRGB(Color=&ColorFromCurve(ColorCurve=DataInterface() {propertyValues {
         AlphaCurve {keys []}
         BlueCurve {keys [<at 0 1>,<at 0.15 10 RCIM_Cubic>,<at 1 1 RCIM_Cubic>]}
         GreenCurve {keys [<at 0 1>, {arriveTangent=-0.1823, interpMode=RCIM_Cubic, time=0.15, value=9.045}, <at 1 0.8177 RCIM_Cubic>]}
         RedCurve {keys [<at 0 1>, {arriveTangent=-0.1812, interpMode=RCIM_Cubic, time=0.15, value=8.5417}, <at 1 0.8188 RCIM_Cubic>]}
        }}))
     Float=1
    )
    ScaleA=true
    ScaleRGB=true
   )
   &CameraOffset(Camera Offset Amount=-5)
  ]
  rendererClasses [&NiagaraSpriteRendererProperties, &NiagaraMeshRendererProperties]
  renderers [
   {rendererClass=&NiagaraSpriteRendererProperties}
   {rendererClass=&NiagaraMeshRendererProperties, rendererIndex=1}
   {rendererClass=&NiagaraMeshRendererProperties, rendererIndex=2}
   {rendererClass=&NiagaraMeshRendererProperties, rendererIndex=3}
  ]
 }
 Lightning_Sprites_Loop_bottom {
  emitterUpdateScript [
   &EmitterState(
    Delay First Loop Only=true
    Inactive Response=::NewEnumerator1
    Life Cycle Mode=::NewEnumerator1
    Loop Delay=0.2
    Loop Duration=1
    Recalculate Duration Each Loop=true
    UseLoopDelay=true
   )
   &SpawnRate(Spawn Probability=0.5, SpawnRate=&UniformRangedFloat(Minimum=0, Maximum=3), Use Spawn Probability=true)
  ]
  particleSpawnScript [
   &InitializeParticle(
    Color=<rgba 0.8317 0 2 1>
    Mesh Scale=<50 50 15>
    Module.Write Color=true
    Module.Write Lifetime=true
    Module.Write Mass=true
    Module.Write Material Random=true
    Module.Write Position=false
    Module.Write Scale=true
    Module.Write SpriteRotation=false
    Module.Write SpriteSize=true
    Position=&SimulationPosition()
    Sprite Rotation=&UniformRangedFloat(Minimum=0, Maximum=360)
    Sprite Size=&UniformRangedVector2D(Minimum=<40 10>, Maximum=<50 50>)
   )
   /SetVariables(Particles.RandomPosition:Vector3f=&UniformRangedVector(Minimum=<0 -30 0>, Maximum=<0 30 0>)) {isSetParameters=true, module=SetVariables_#1}
   &DynamicMaterialParameters(
    Index 0 Param 3=&FloatFromCurve(FloatCurve=DataInterface() {propertyValues {Curve {keys [<at 0 1 RCIM_Cubic>,<at 0.2 0.025 RCIM_Cubic>,<at 1 1>]}}})
    Module.Param0WriteEnabled=false
    Module.Param1WriteEnabled=false
    Module.Param2WriteEnabled=true
    Module.Param3WriteEnabled=false
   )
   &FN_SetSpawnGroup_v2(Number of Spawn Groups=5)
   /SetVariables(Particles.AirGlow_Color:LinearColor=<rgba 4.1587 0 10 1>, Particles.AirGlow_Size:Vector2f=<150 150>) {isSetParameters=true, module=SetVariables_#3}
   /SetVariables(Particles.SubImageIndex=1) {isSetParameters=true, module=SetVariables_#4}
   /SetVariables(Particles.RandomRot:Quat4f=&MakeQuatFromAxisAngle(Axis=<0 0 1>, AngleInDegrees=&UniformRangedFloat(Minimum=-45, Maximum=45))) {isSetParameters=true, module=SetVariables_#5}
   /SetVariables(
    Particles.MeshOrientation=&MultiplyQuaternion(
     Quaternion A=$Particles.RandomRot
     Quaternion B=&FindQuatBetweenTwoVectors(From Vector=<0 -1 0>, To Vector=<0 0 1>)
    )
    Particles.MeshOrientation2:Quat4f=&MultiplyQuaternion(
     Quaternion A=$Particles.RandomRot
     Quaternion B=&FindQuatBetweenTwoVectors(From Vector=<0 -1 0>, To Vector=<0 0 -1>)
    )
   ) {isSetParameters=true, module=SetVariables_#6}
  ]
  particleUpdateScript [
   &ParticleState()
   &SkeletalMeshLocation(
    Bone / Socket Index=$Particles.SpawnGroup
    Inherit Velocity=false
    MaskSpawnGroup=false
    Mesh Sampling Type=::NewEnumerator5
    Sampled Position Offset=&Add_Vector(A=$Particles.RandomPosition, B=<0 0 -5>)
    Sampled Position Offset Coordinate Space:ENiagaraMeshTransforms=::NewEnumerator3
    Skeletal Mesh=$User.SkeletalMesh2
    Socket Sampling Mode=::NewEnumerator5
   )
   /SetVariables(
    Particles.Decal_CameraOffset=5
    Particles.Decal_Color:LinearColor=&ColorFromCurve(ColorCurve=DataInterface() {propertyValues {
       AlphaCurve {keys [<at 0 0 RCIM_Cubic>,<at 0.05 0.5>,<at 0.25 0.5>,<at 0.5 0 RCIM_Cubic>]}
       BlueCurve {keys [<at 0 2 RCIM_Cubic>,<at 0.25 1>]}
       GreenCurve {keys [<at 0 0 RCIM_Cubic>,<at 0.25 0>]}
       RedCurve {keys [<at 0 1.5807 RCIM_Cubic>,<at 0.25 0.7904>]}
      }})
    Particles.Decal_DynamicParameter:Vector4f=&MakeVector4(X=4, Y=1, Z=1, W=1)
    Particles.Decal_LT_Position:Vector3f=&Add_Vector(
     A=Linked() {linkedVariable {name=Particles.Position, type=NiagaraPosition}}
     B=&RotateVectorByQuaternion(Vector To Rotate=<-15 0 0>, Quaternion=$Particles.RandomRot)
    )
    Particles.Decal_RT_Position:Vector3f=&Add_Vector(
     A=Linked() {linkedVariable {name=Particles.Position, type=NiagaraPosition}}
     B=&RotateVectorByQuaternion(Vector To Rotate=<12 0 0>, Quaternion=$Particles.RandomRot)
    )
    Particles.Decal_Size:Vector3f=<20 20 0>
   ) {isSetParameters=true, module=SetVariables_#7}
   /SetVariables(Particles.SpriteFacing=&RotateVectorByQuaternion(Vector To Rotate=<1 0 0>, Quaternion=$Output.SkeletalMeshLocation.SampledRotation)) {isSetParameters=true, module=SetVariables_#2}
   &DynamicMaterialParameters(
    Index 0 Param 3=&FloatFromCurve(FloatCurve=DataInterface() {propertyValues {Curve {keys [<at 0 1 RCIM_Cubic>,<at 0.05 0 RCIM_Cubic>,<at 1 1 RCIM_Cubic>]}}})
    Module.Param0WriteEnabled=false
    Module.Param1WriteEnabled=false
    Module.Param2WriteEnabled=true
    Module.Param3WriteEnabled=false
   ) {module=DynamicMaterialParameters001}
   &ScaleColor(
    Scale Alpha=&FloatFromCurve(FloatCurve=DataInterface() {propertyValues {Curve {keys [<at 0.75 1 RCIM_Cubic>, {arriveTangent=-10.3917, interpMode=RCIM_Cubic, leaveTangent=-10.3917, tangentMode=RCTM_User, time=1, value=0}]}}})
    Scale RGB=&Multiply_VectorByFloat(
     Vector=&MakeVectorFromLinearColorRGB(Color=&ColorFromCurve(ColorCurve=DataInterface() {propertyValues {
         AlphaCurve {keys []}
         BlueCurve {keys [<at 0 1>,<at 0.15 10 RCIM_Cubic>,<at 1 1 RCIM_Cubic>]}
         GreenCurve {keys [<at 0 1>, {arriveTangent=-0.1823, interpMode=RCIM_Cubic, time=0.15, value=9.045}, <at 1 0.8177 RCIM_Cubic>]}
         RedCurve {keys [<at 0 1>, {arriveTangent=-0.1812, interpMode=RCIM_Cubic, time=0.15, value=8.5417}, <at 1 0.8188 RCIM_Cubic>]}
        }}))
     Float=1
    )
    ScaleA=true
    ScaleRGB=true
   )
   &CameraOffset(Camera Offset Amount=-5)
  ]
  rendererClasses [&NiagaraSpriteRendererProperties, &NiagaraMeshRendererProperties]
  renderers [
   {rendererClass=&NiagaraSpriteRendererProperties}
   {rendererClass=&NiagaraMeshRendererProperties, rendererIndex=1}
   {rendererClass=&NiagaraMeshRendererProperties, rendererIndex=2}
   {rendererClass=&NiagaraMeshRendererProperties, rendererIndex=3}
  ]
 }
 MeshTrails {
  emitterSpawnScript [&FN_PreserveSkydiveUserVariables() {module=PreserveUserVariables}]
  emitterUpdateScript [
   &EmitterState()
   &SpawnBurst_Instantaneous(
    Spawn Count=HlslExpression() {hlslExpression="Emitter.LoopCount == 0"}
    Spawn Probability=&FN_FrontEndFloatModifier(
     Gameplay=&MakeCustomFloatFromBool(Bool=$User.bIsFullyDeployed, True Float=1, False Float=0)
     Front End=1
    )
    Use Spawn Probability=true
   ) {module=SpawnBurst_Instantaneous001}
  ]
  particleSpawnScript [
   &InitializeParticle(
    Color=<rgba 2.8457 0.6 6 0.2>
    Module.Write Color=true
    Module.Write Lifetime=true
    Module.Write Mass=false
    Module.Write Material Random=true
    Module.Write Position=false
    Module.Write Scale=false
    Module.Write SpriteRotation=false
    Module.Write SpriteSize=false
   )
   &self:Mesh_Trail_Setup(
    Full Trail Length=500
    Long Trail Width:NiagaraFloat=&FN_FrontEndFloatModifier(Gameplay=85, Front End=60)
    Short Trail Multiplier=0.35
    ShortTrail Width:NiagaraFloat=&FN_FrontEndFloatModifier(Gameplay=80, Front End=60)
   ) {module=Mesh_Trail_Setup}
  ]
  particleUpdateScript [
   &ParticleState(Kill Particles When Lifetime Has Elapsed=false, Let Infinitely Lived Particles Die When Emitter Deactivates=true)
   &SkeletalMeshLocation(
    Apply Sampled Orientation to Mesh Particles=true
    Bone / Socket Index=2
    Inherit Velocity=false
    Interpolate Mesh Sampling=false
    MaskSpawnGroup=false
    Mesh Sampling Type=::NewEnumerator5
    Sampled Position Offset Coordinate Space:ENiagaraMeshTransforms=::NewEnumerator3
    Skeletal Mesh=$User.SkeletalMesh
    Socket Sampling Mode=::NewEnumerator5
    Write Sampled Mesh Attrributes to Particle Attributes=true
   )
   /SetVariables(Particles.MeshTwinOrientation:Quat4f=&MultiplyQuaternion(Quaternion A=$Particles.MeshOrientation, Quaternion B=<1 0 0 0>)) {isSetParameters=true}
   /SetVariables(Particles.NormalishSocketVelocity:NiagaraFloat=&Multiply_Float(
     A=-0.004
     B=&ClampFloat(
      Float=&MakeFloatFromVector(Vector=$Particles.SkeletalMeshLocation.SampledVelocity, Channel=::NewEnumerator0)
      Min=-250
      Max=250
     )
    )) {isSetParameters=true}
   /SetVariables(Particles.VelocityOutputControl:NiagaraFloat=&Subtract_Float(
     A=2.7
     B=&Multiply_Float(A=2.2725, B=&Multiply_Float(A=$Particles.NormalishSocketVelocity, B=$Particles.NormalishSocketVelocity))
    )) {isSetParameters=true}
   /SetVariables(Particles.SocketSpeed:NiagaraFloat=&VectorLength(Vector=$StackContext.SkeletalMeshLocation.SampledVelocity)) {isSetParameters=true}
   &self:Exhaust_to_DynaMat(
    ExhaustVector:Vector3f=&FN_FrontEndVector3Modifier(
     Gameplay=$User.CurrentExhaust
     Front End=&Multiply_VectorByFloat(
      Vector=<0 10 0>
      Float=&Multiply_Float(A=$Particles.NormalishSocketVelocity, B=&Power(a=$Particles.VelocityOutputControl, b=0.8))
     )
    )
    SpeedMult=1
   ) {module=Exhaust_to_DynaMat}
   &self:Mesh_Trail_Length_Update(Dynamic Length Scale:NiagaraFloat=&FN_FrontEndFloatModifier(Gameplay=$User.TrailLengthScale, Front End=1)) {module=Mesh_Trail_Length_Update}
   &ScaleColor(
    Scale Alpha=&FN_FrontEndFloatModifier(Gameplay=$User.BPOpacityMult, Front End=1)
    ScaleA=true
    ScaleRGB=true
   )
  ]
  rendererClasses [&NiagaraMeshRendererProperties]
  renderers [
   {rendererClass=&NiagaraMeshRendererProperties, rendererIndex=0}
   {rendererClass=&NiagaraMeshRendererProperties, rendererIndex=1}
   {rendererClass=&NiagaraMeshRendererProperties, rendererIndex=2}
  ]
 }
}
summary {
 systemName=NS_KevinCouture_Glider_Trail
 emitters [
  {
   emitterName=Lightning_Mesh_Loop
   rendererClasses [&NiagaraMeshRendererProperties, &NiagaraSpriteRendererProperties]
  }
  {
   emitterName=Lightning_Sprites_Loop
   rendererClasses [&NiagaraSpriteRendererProperties, &NiagaraMeshRendererProperties]
  }
  {emitterName=Decal, rendererClasses [&NiagaraMeshRendererProperties]}
  {
   emitterName=Lightning_Sprites_Loop_bottom
   rendererClasses [&NiagaraSpriteRendererProperties, &NiagaraMeshRendererProperties]
  }
  {emitterName=Light, rendererClasses [&NiagaraLightRendererProperties]}
  {emitterName=BurstRunes, rendererClasses [&NiagaraSpriteRendererProperties]}
  {emitterName=Feet_Decal, rendererClasses [&NiagaraMeshRendererProperties]}
  {emitterName=MeshTrails, rendererClasses [&NiagaraMeshRendererProperties]}
  {emitterName=Lightning_Sprites_Burst, rendererClasses [&NiagaraSpriteRendererProperties]}
 ]
 userVariables [
  User.bIsFrontEnd {type=NiagaraBool}
  User.bIsFrontEndPreview {type=NiagaraBool}
  User.bIsFullyDeployed {type=NiagaraBool, defaultValue=true}
  User.BPOpacityMult {type=NiagaraFloat, defaultValue=0, description=""}
  User.CurrentExhaust {type=Vector3f, description="", defaultValue {struct=Vector3f, value=<0 0 0>}}
  User.ForwardDot {type=NiagaraFloat, defaultValue=0, description=""}
  User.LocalAlpha {type=NiagaraFloat, defaultValue=0, description=""}
  User.PlayerSpeed {type=NiagaraFloat, defaultValue=0, description=""}
  User.RibbonIDOffset {type=NiagaraInt32}
  User.SkeletalMesh {
   type=NiagaraDataInterfaceSkeletalMesh
   defaultValue {
    struct=NiagaraExt_VariableValue_DataInterface
    value {dataInterface=&self:NiagaraDataInterfaceSkeletalMesh_0}
   }
  }
  User.SkeletalMesh2 {
   type=NiagaraDataInterfaceSkeletalMesh
   defaultValue {
    struct=NiagaraExt_VariableValue_DataInterface
    value {dataInterface=&self:NiagaraDataInterfaceSkeletalMesh_1}
   }
  }
  User.TrailLengthScale {type=NiagaraFloat, defaultValue=0, description=""}
 ]
}
